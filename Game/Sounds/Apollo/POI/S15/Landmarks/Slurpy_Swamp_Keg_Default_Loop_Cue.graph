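# Looping ambient prop Sound Cue for the Slurpy Swamp keg landmark, playing a building-topper loop.
# A single looping Wave Player is the root, routed to the Athena ambient props sound class at slightly reduced volume.
# Spatialization references the keg's shared attenuation asset; stored box-shaped, occluded attenuation overrides are not switched on.
# A concurrency override caps simultaneous instances, with volume scaling by distance.
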SoundCue /Game/Sounds/Apollo/POI/S15/Landmarks/Slurpy_Swamp_Keg_Default_Loop_Cue
root=WavePlayer_0
nodes [{id=WavePlayer_0, properties {bLooping=true, soundWaveAssetPtr=&/Slurpy_Swamp_Keg_Building_Topper_Loop}}]
properties {
 attenuationSettings=&/Slurpy_Swamp_Keg_Default_Att
 bOverrideConcurrency=true
 maxDistance=2000
 soundClassObject=&Fort_Ambient_Props_Athena
 volumeMultiplier=0.928
 attenuationOverrides {
  attenuationShape=Box
  bEnableOcclusion=true
  falloffDistance=740
  occlusionInterpolationTime=2
  occlusionLowPassFilterFrequency=600
  occlusionTraceChannel=ECC_GameTraceChannel11
  reverbWetLevelMin=0.6905
  stereoSpread=5000
  attenuationShapeExtents=<600 600 800>
 }
 concurrencyOverrides {maxCount=3, volumeScaleMode=Distance}
}
